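; Weapon Blueprint for the Unstable FireIce pump shotgun, handling its alternating fire and ice muzzle effects and sounds.
; It starts on ice and flips element when ammo is consumed, using DoOnce and a short timer to gate toggles.
; Each flip swaps the active muzzle component, with DoOnce-gated attached sounds.
; Firing passes the element as FireIceIndex to the muzzle Niagara system and a spawned blast overlay; unequipping fades the stored sound.
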(event EventOnPlayWeaponFireFX
  (params (bPersistentFire :type Boolean) (bSecondaryFire :type Boolean))
  (bind _isiceeffectactive (Variables|Default|GetIsIceEffectActive?))
  (bind _returnvalue (Math|Integer|SelectInt 0 1 _isiceeffectactive))
  (|CallParentFunction bPersistentFire bSecondaryFire)
  (Utilities|FlowControl|Sequence
    (:then_0
      (bind _spawned_muzzle_system (Variables|Effects|Muzzle|GetSpawned%20Muzzle%20System)
        (:then
          (bind _asniagara_particle_system_component (Utilities|Casting|CastToNiagaraComponent _spawned_muzzle_system))
          (Niagara|SetNiagaraVariable(Int32) _asniagara_particle_system_component "FireIceIndex" _returnvalue))
        (:else)))
    (:then_1
      (bind _returnvalue_2 (Niagara|SpawnSystemAttached "/FirePetalWeaponGameplay/Gameplay/FirePetal_PumpShotgun/FireIce/FX/NS_Unstable_FireIce_MuzzleBlastOverlay.NS_Unstable_FireIce_MuzzleBlastOverlay" (Variables|Visual|GetWeaponMesh) (Effects|GetMuzzleSocketName 0) "0, 0, 0" "0, 0, 0" "SnapToTarget" true))
      (Niagara|SetNiagaraVariable(Int32) _returnvalue_2 "FireIceIndex" _returnvalue))))

(event EventBeginPlay
  (EventDispatchers|BindEventtoOnConsumeAmmo (delegate Custom|OnConsumeAmmo_Event))
  (Variables|Default|SetIsIceEffectActive? true)
  (CallFunction|EnableIceFX))

(event Custom|OnConsumeAmmo_Event
  (params (Quantity :type Integer))
  (Utilities|FlowControl|DoOnce false
    (:Completed
      (if (Variables|Default|GetIsIceEffectActive?)
        (CallFunction|EnableFireFX)
        (else
          (CallFunction|EnableIceFX)))
      (Utilities|Time|SetTimerbyEvent (delegate Custom|Reset) 0.2))))

(event Custom|Reset
  (Utilities|FlowControl|DoOnce false
    (:Completed
      (if (Variables|Default|GetIsIceEffectActive?)
        (CallFunction|EnableFireFX)
        (else
          (CallFunction|EnableIceFX)))
      (Utilities|Time|SetTimerbyEvent (delegate Custom|Reset) 0.2))))

(event Custom|EnableFireFX
  (bind _ns_firemuzzle_fx (Variables|Default|GetNS_FireMuzzle_FX))
  (Components|Activation|Deactivate (Variables|Default|GetNS_IceMuzzle_FX))
  (Components|Activation|Activate _ns_firemuzzle_fx true)
  (Variables|Default|SetIsIceEffectActive? false)
  (Utilities|FlowControl|Sequence
    (:then_0
      (Utilities|FlowControl|DoOnce false
        (:Completed
          (bind _returnvalue (Audio|SpawnSoundAttached :Sound "/FirePetalWeaponGameplay/Gameplay/FirePetal_PumpShotgun/FireIce/Sounds/SC_Shotgun_Unstable_FireIce_ApplyFire.SC_Shotgun_Unstable_FireIce_ApplyFire" :AttachToComponent _ns_firemuzzle_fx :AttachPointName "None" :LocationType "KeepRelativeOffset" :bStopWhenAttachedToDestroyed true :VolumeMultiplier 1.0 :PitchMultiplier 1.0 :StartTime 0.0 :bAutoDestroy true))
          (Variables|Default|SetMuzzle_SFX _returnvalue))))
    (:then_1
      (Utilities|FlowControl|DoOnce false
        (:Completed
          (bind _returnvalue_2 (Audio|SpawnSoundAttached :Sound "/FirePetalWeaponGameplay/Gameplay/FirePetal_PumpShotgun/FireIce/Sounds/SC_Shotgun_Unstable_FireIce_ApplyIce.SC_Shotgun_Unstable_FireIce_ApplyIce" :AttachToComponent (Variables|Default|GetNS_IceMuzzle_FX) :AttachPointName "None" :LocationType "KeepRelativeOffset" :bStopWhenAttachedToDestroyed true :VolumeMultiplier 1.0 :PitchMultiplier 1.0 :StartTime 0.0 :bAutoDestroy true))
          (Variables|Default|SetMuzzle_SFX _returnvalue_2))))))

(event Custom|EnableIceFX
  (bind _ns_icemuzzle_fx (Variables|Default|GetNS_IceMuzzle_FX))
  (Components|Activation|Deactivate (Variables|Default|GetNS_FireMuzzle_FX))
  (Components|Activation|Activate _ns_icemuzzle_fx true)
  (Variables|Default|SetIsIceEffectActive? true)
  (Utilities|FlowControl|Sequence
    (:then_0
      (Utilities|FlowControl|DoOnce false
        (:Completed
          (bind _returnvalue (Audio|SpawnSoundAttached :Sound "/FirePetalWeaponGameplay/Gameplay/FirePetal_PumpShotgun/FireIce/Sounds/SC_Shotgun_Unstable_FireIce_ApplyIce.SC_Shotgun_Unstable_FireIce_ApplyIce" :AttachToComponent _ns_icemuzzle_fx :AttachPointName "None" :LocationType "KeepRelativeOffset" :bStopWhenAttachedToDestroyed true :VolumeMultiplier 1.0 :PitchMultiplier 1.0 :StartTime 0.0 :bAutoDestroy true))
          (Variables|Default|SetMuzzle_SFX _returnvalue))))
    (:then_1
      (Utilities|FlowControl|DoOnce false
        (:Completed
          (bind _returnvalue_2 (Audio|SpawnSoundAttached :Sound "/FirePetalWeaponGameplay/Gameplay/FirePetal_PumpShotgun/FireIce/Sounds/SC_Shotgun_Unstable_FireIce_ApplyFire.SC_Shotgun_Unstable_FireIce_ApplyFire" :AttachToComponent (Variables|Default|GetNS_FireMuzzle_FX) :AttachPointName "None" :LocationType "KeepRelativeOffset" :bStopWhenAttachedToDestroyed true :VolumeMultiplier 1.0 :PitchMultiplier 1.0 :StartTime 0.0 :bAutoDestroy true))
          (Variables|Default|SetMuzzle_SFX _returnvalue_2))))))

(event Custom|DisableAllFX
  (Components|Activation|Deactivate (Variables|Default|GetNS_FireMuzzle_FX)))

(event FortWeapon|EventOnUnEquippedAndDetached
  (|CallParentFunction)
  (Audio|Components|Audio|FadeOut (Variables|Default|GetMuzzle_SFX) 0.3 0.0 "Linear" :!class /Script/Engine.AudioComponent))

(fn GetActorAudioParameters
  (return 0))

(fn ConstructionScript
  (|CallParentFunction))
