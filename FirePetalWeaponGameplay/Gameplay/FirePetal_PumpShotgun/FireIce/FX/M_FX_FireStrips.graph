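# Additive unlit Niagara mesh-particle material for fire strips in the FirePetal pump shotgun's FireIce effects.
# Panned flame noise, with time offset per strand by vertex color and a dynamic parameter, adds highlights and varies an end-to-core color gradient.
# A base-to-tip, sphere and Fresnel mask drives that gradient and feeds a smoothstep erosion positioned by dynamic parameter.
# A second panned noise scales a push along vertex normals.
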Material /FirePetalWeaponGameplay/Gameplay/FirePetal_PumpShotgun/FireIce/FX/M_FX_FireStrips
graph {
 properties {
  bAutomaticallySetUsageInEditor=false
  bUsedWithNiagaraMeshParticles=true
  blendMode=BLEND_Additive
  shadingModel=MSM_Unlit
  usedShadingModels=Unlit
 }
 nodes [
  n1 = ScalarParameter() {group=Highlights, parameterName=Highlights_P}
  n2 = ScalarParameter() {group=Highlights, parameterName=Highlights_W}
  n3 = ComponentMask(None=TextureCoordinate()) {g=true, r=true}
  n4 = VectorParameter() {group=Detail, parameterName=Panner_Scale, defaultValue=<rgba 1 1 0 1>}
  n5 = Multiply(A=ComponentMask(None=@n4.RGB) {g=true, r=true}, B=@n4.B)
  n6 = VectorParameter() {group=VS, parameterName=Panner_Speed, defaultValue=<rgba 1 1 0 1>}
  n7 = Multiply(A=ComponentMask(None=@n6.RGB) {g=true, r=true}, B=@n4.B)
  n8 = TextureSampleParameter2D(UVs=Panner(Coordinate=Multiply(A=@n3, B=@n5), Time=NamedRerouteUsage(), Speed=Multiply(A=@n7, B=@n6.B)) {bFractionalPart=true}) {
   parameterName=T_fireNoise
   samplerType=SAMPLERTYPE_LinearGrayscale
   texture=/Game/Effects/Fort_Effects/Emotes/UltraLight/Texture/T_FluidFlameTile_Greyscale
  }
  n9 = Multiply(
   A=SmoothStep(Min=@n1, Max=Add(A=@n1, B=@n2), Value=NamedRerouteDeclaration(None=@n8.R) {name=FireNoise, nodeColor=<rgba 0 1 0.6484 1>})
   B=VectorParameter() {group=Color, parameterName=HighlightsColor, defaultValue=<rgba 0 0 0 1>}.RGB
  )
  n10 = NamedRerouteDeclaration(None=ComponentMask(None=TextureCoordinate()) {r=true}) {name=BaseToTip, nodeColor=<rgba 0.7109 1 0 1>}
  n11 = TextureCoordinate()
  n12 = VectorParameter() {parameterName=MaskPosition, defaultValue=<rgba 0 0 0 1>}
  n13 = ComponentMask(None=@n12.RGB) {g=true}
  n14 = ScalarParameter() {parameterName=EdgeHardness}
  n15 = ScalarParameter() {defaultValue=1, group=Fresnel, parameterName=FresnelBrightness}
  n16 = ScalarParameter() {group=Fresnel, parameterName=FresnelExp}
  n17 = Multiply(
   A=Multiply(A=OneMinus(None=@n10), B=SphereMask(A=@n11, B=@n13, Radius=@n12.B, Hardness=@n14) {hardnessPercent=50})
   B=Saturate(None=MaterialFunctionCall(Brightness=@n15, Exponent=@n16) {materialFunction=&MF_Fresnel_FX}.Fresnel Double Sided)
  )
  n18 = Power(Base=@n17) {constExponent=4}
  n19 = Multiply(A=Subtract(A=Multiply(A=@n8.R) {constB=2}), B=ScalarParameter() {parameterName=ColorNoise})
  n20 = LinearInterpolate(
   A=VectorParameter() {group=Color, parameterName=EndColor, defaultValue=<rgba 0 0 0 1>}.RGB
   B=VectorParameter() {group=Color, parameterName=CoreColor, defaultValue=<rgba 0 0 0 1>}.RGB
   Alpha=Saturate(None=Add(A=@n18, B=@n19))
  )
  MaterialFunctionCall(Emissive=Multiply(A=ParticleColor().RGB, B=Add(A=@n9, B=@n20))) {materialFunction=&MF_FX_EmissiveAdjustments}
  n21 = ComponentMask(None=TextureCoordinate()) {g=true, r=true}
  n22 = VectorParameter() {group=VS, parameterName=VS_Panner_Scale, defaultValue=<rgba 1 1 0 1>}
  n23 = Multiply(A=ComponentMask(None=@n22.RGB) {g=true, r=true}, B=@n22.B)
  n24 = VectorParameter() {group=VS, parameterName=VS_Panner_Speed, defaultValue=<rgba 1 1 0 1>}
  n25 = Multiply(A=ComponentMask(None=@n24.RGB) {g=true, r=true}, B=@n22.B)
  n26 = TextureSampleParameter2D(UVs=Panner(Coordinate=Multiply(A=@n21, B=@n23), Time=NamedRerouteUsage(), Speed=Multiply(A=@n25, B=@n24.B)) {bFractionalPart=true}) {
   parameterName=T_PushNoise
   samplerType=SAMPLERTYPE_LinearGrayscale
   texture=&T_FluidFlameTile_Greyscale
  }
  n27 = ScalarParameter() {parameterName=VS_Push}
  Multiply(A=VertexNormalWS(), B=Multiply(A=NamedRerouteUsage(), B=Multiply(A=@n26.R, B=@n27)))
  n28 = Time()
  n29 = VertexColor()
  NamedRerouteDeclaration(None=Add(
    A=Add(A=@n28, B=@n29.R)
    B=DynamicParameter() {defaultValue=<rgba 1 0.064 1 1>, paramNames [TimeOffset, Erode_Position, Param3, Param4]}.TimeOffset
   )) {name=PerStrandTime, nodeColor=<rgba 0 0.4844 1 1>}
  n30 = Add(A=ComponentMask(None=TextureCoordinate()) {r=true}, B=ScalarParameter() {group=StartingEdge, parameterName=StartingEdge_P})
  n31 = ScalarParameter() {group=StartingEdge, parameterName=StartingEdge_Str}
  NamedRerouteDeclaration(None=Clamp(None=Multiply(A=@n30, B=@n31))) {name=StartEdgeMask, nodeColor=<rgba 0.7109 1 0 1>}
  NamedRerouteUsage()
  NamedRerouteUsage()
  NamedRerouteUsage()
  NamedRerouteUsage()
  ScalarParameter() {parameterName=Erode_P}
  n32 = Add(
   A=NamedRerouteUsage()
   B=DynamicParameter() {defaultValue=<rgba 1 0.064 1 1>, paramNames [TimeOffset, Erode_Position, Param3, Param4]}.Erode_Position
  )
  n33 = NamedRerouteUsage()
  SmoothStep(
   Min=@n32
   Max=Add(A=@n32, B=ScalarParameter() {parameterName=Erode_W})
   Value=NamedRerouteDeclaration(None=Multiply(A=@n33, B=@n17)) {name=Mask, nodeColor=<rgba 0 1 0.4844 1>}
  )
  Subtract(A=Multiply(B=@n26.R) {constA=2})
  VectorParameter() {group=Color, parameterName=EndColor, defaultValue=<rgba 0 0 0 1>}
 ]
}
