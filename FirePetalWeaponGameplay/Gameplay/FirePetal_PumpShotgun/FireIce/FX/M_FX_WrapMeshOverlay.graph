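# Additive unlit material for Niagara mesh particles, an animated noisy overlay in the FireIce pump shotgun effects.
# Two panning flame-noise samples, one warping the other's UVs, distort a U gradient that blends EndColor to CoreColor, tinted by particle color.
# Opacity combines that gradient with particle alpha, vertex red and inverted depth fade, favoring areas near scene intersections.
# A Reveal dynamic parameter drives a noisy wipe along U, and a static switch offers a modulate-style color output.
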Material /FirePetalWeaponGameplay/Gameplay/FirePetal_PumpShotgun/FireIce/FX/M_FX_WrapMeshOverlay
graph {
 properties {
  bAutomaticallySetUsageInEditor=false
  bUsedWithNiagaraMeshParticles=true
  blendMode=BLEND_Additive
  shadingModel=MSM_Unlit
  usedShadingModels=Unlit
 }
 nodes [
  Multiply()
  n1 = DepthFade(FadeDistance=ScalarParameter() {parameterName=DepthDistance})
  n2 = ParticleColor()
  n3 = Add(A=ComponentMask(None=TextureCoordinate()) {r=true}, B=ScalarParameter() {parameterName=GradientBias})
  n4 = ScalarParameter() {defaultValue=1, parameterName=GradientScale}
  n5 = ComponentMask(None=TextureCoordinate()) {g=true, r=true}
  n6 = ComponentMask(None=VectorParameter() {group=Detail, parameterName=Panner_Scale_2, defaultValue=<rgba 1 1 0 1>}.RGB) {g=true, r=true}
  n7 = VectorParameter() {group=Detail, parameterName=Panner_Speed_2, defaultValue=<rgba 1 1 0 1>}
  n8 = TextureSampleParameter2D(UVs=Panner(Coordinate=Multiply(A=@n5, B=@n6), Speed=ComponentMask(None=@n7.RGB) {g=true, r=true}) {bFractionalPart=true}) {
   parameterName=T_FireDetail
   samplerType=SAMPLERTYPE_LinearGrayscale
   texture=/Game/Effects/Fort_Effects/Emotes/UltraLight/Texture/T_FluidFlameTile_Greyscale
  }
  n9 = ScalarParameter() {parameterName=Noise_Offset}
  n10 = Multiply(
   A=ComponentMask(None=TextureCoordinate()) {g=true, r=true}
   B=ComponentMask(None=VectorParameter() {group=Detail, parameterName=Panner_Scale, defaultValue=<rgba 1 1 0 1>}.RGB) {g=true, r=true}
  )
  n11 = ComponentMask(None=VectorParameter() {group=Detail, parameterName=Panner_Speed, defaultValue=<rgba 1 1 0 1>}.RGB) {g=true, r=true}
  n12 = TextureSampleParameter2D(UVs=Add(A=Multiply(A=@n8.R, B=@n9), B=Panner(Coordinate=@n10, Speed=@n11) {bFractionalPart=true})) {
   parameterName=T_FireDetail
   samplerType=SAMPLERTYPE_LinearGrayscale
   texture=&T_FluidFlameTile_Greyscale
  }
  n13 = Subtract(A=Multiply(A=@n12.R) {constB=2})
  n14 = ScalarParameter() {group=Detail, parameterName=ColorDistortionStrength_End}
  n15 = ScalarParameter() {group=Detail, parameterName=ColorDistortionStrength_Tip}
  n16 = ComponentMask(None=TextureCoordinate()) {r=true}
  n17 = Add(A=Clamp(None=Multiply(A=@n3, B=@n4)), B=Multiply(A=@n13, B=LinearInterpolate(A=@n14, B=@n15, Alpha=@n16)))
  n18 = Saturate(None=Multiply(A=OneMinus(None=@n1), B=Multiply(A=@n2.A, B=@n17)))
  n19 = VertexColor()
  n20 = DynamicParameter() {paramNames [Reveal,Param2,Param3,Param4]}
  n21 = Add(A=@n20.Reveal, B=ScalarParameter() {parameterName=RevealW})
  n22 = Add(A=ComponentMask(None=TextureCoordinate()) {r=true}, B=Multiply(A=NamedRerouteUsage(), B=ScalarParameter() {parameterName=NoiseRevealAmnt}))
  NamedRerouteDeclaration(None=Multiply(A=Multiply(A=@n18, B=@n19.R), B=SmoothStep(Min=@n20.Reveal, Max=@n21, Value=@n22))) {name=Opacity, nodeColor=<rgba 0.2422 0 1 1>}
  NamedRerouteDeclaration(None=@n13) {name=ExpandedNoise, nodeColor=<rgba 1 0.375 0 1>}
  Power(Base=@n19.R)
  n23 = VectorParameter() {parameterName=EndColor, defaultValue=<rgba 0 0 0 1>}
  n24 = VectorParameter() {parameterName=CoreColor, defaultValue=<rgba 0 0 0 1>}
  n25 = Power(Base=@n17, Exp=ScalarParameter() {defaultValue=2, parameterName=Falloff})
  n26 = MaterialFunctionCall(Emissive=Multiply(A=LinearInterpolate(A=@n23.RGB, B=@n24.RGB, Alpha=@n25), B=@n2.RGB)) {materialFunction=&MF_FX_EmissiveAdjustments}
  StaticSwitchParameter(True=LinearInterpolate(B=@n26.Result, Alpha=NamedRerouteUsage()) {constA=1}, False=@n26.Result) {parameterName=Modulate}
 ]
}
