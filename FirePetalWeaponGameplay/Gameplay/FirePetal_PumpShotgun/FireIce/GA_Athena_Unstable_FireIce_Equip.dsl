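; Gameplay ability for the unstable Fire/Ice pump shotgun that applies an ice or spicy effect when the weapon deals damage.
; An activation event confirms the current weapon is the damage source and keeps only hostile player pawns or Deimos-tagged creatures.
; The weapon's IsIceEffectActive flag picks the effect: true gives Spicy to the shooter, false gives Ice to the target.
; Effect cycling here is deprecated.
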(event Ability|EventActivateAbilityFromEvent
  (params (EventData :type Gameplay%20Event%20Data%20Structure%20%28by%20ref%29))
  (bind _output_get (Variables|Default|SetEvent%20Data EventData))
  (if (Default|IsDamgefromSourceWeapon _output_get)
    (bind _outhitactors (Utilities|HasValidGroupOfHitActors (CallFunction|GetHitActorsfromEventData _output_get))
      (:true
        (Variables|Default|SetHitActors _outhitactors)
        (CallFunction|EventApplyEffect))
      (:false
        _))
    (Ability|EndAbility)
    (else
      _)))

(event Custom|EventApplyEffect
  (bind _iceeffectactive (CallFunction|GetCurrentEffectfromParent))
  (if _iceeffectactive
    (CallFunction|ApplySpicy)
    (else
      (CallFunction|ApplyIce))))

(event Custom|EventCycleEffects
  ; Deprecated
  (bind _iceeffectactive (CallFunction|GetCurrentEffectfromParent))
  (if _iceeffectactive
    (Variables|Default|SetCurrentNextEffect (Variables|Default|GetSpicyGameplayEffectClass))
    (else
      (Variables|Default|SetCurrentNextEffect (Variables|Default|GetIceGameplayEffectClass)))))

(fn ApplyIce
  (for _array_element (Variables|Default|GetHitActors)
    (bind _fortpawn (Class|FortQuestBlueprintLibrary|GetFortPawnFromObject _array_element)
      (:Valid
        (GameplayEffects|ApplyGameplayEffectToTarget :self (Ability|GetAbilitySystemComponent (Ability|GetAvatarActorfromActorInfo)) :GameplayEffectClass (Variables|Default|GetIceGameplayEffectClass) :Target (Ability|GetAbilitySystemComponent _fortpawn) :Level 0.0))))
  (return))

(fn ApplySpicy
  (GameplayEffects|ApplyGameplayEffectToTarget :self (Ability|GetAbilitySystemComponent (Ability|GetAvatarActorfromActorInfo)) :GameplayEffectClass (Variables|Default|GetSpicyGameplayEffectClass) :Target (Ability|GetAbilitySystemComponent (Ability|GetActivatingPawn)) :Level 0.0))

(fn CanActivateProcFromPawn
  (params (HitActor :type Actor%20Object%20Reference))
  (if (Utilities|IsA HitActor "/Script/FortniteGame.FortPawn")
    (if (Utilities|IsA HitActor "/Script/FortniteGame.FortPlayerPawn")
      (return true)
      (else
        (bind _returnvalue (GameplayTags|HasMatchingGameplayTag(Message) HitActor (Variables|Default|GetDeimosTag))) ; Special ability can also trigger when hitting StW Creatures
        (return _returnvalue)))
    (else
      (return false))))

(fn GetAvatarAsFortPawn
  (bind _asfort_pawn (Utilities|Casting|CastToFortPawn (Ability|GetAvatarActorfromActorInfo))
    (:then
      (return _asfort_pawn))
    (:CastFailed
      (return 0))))

(fn GetAvatarController
  (bind _outfortplayerpawn (CallFunction|GetAvatarasValidFortPlayerPawn)
    (:ValidFortPlayerPawn
      (Utilities|IsValid (Fortnite|GetFortPlayerControllerfromActor _outfortplayerpawn)
        (:Is%20Valid
          (return (Fortnite|GetFortPlayerControllerfromActor _outfortplayerpawn)))
        (:Is%20Not%20Valid
          (Utilities|IsValid (Pawn|GetController _outfortplayerpawn)
            (:Is%20Valid
              (return (Pawn|GetController _outfortplayerpawn)))
            (:Is%20Not%20Valid
              _)))))
    (:AvatarCastFailed
      _))
  (return 0))

(fn GetCurrentEffectFromParent
  (bind _asb_shotgun_pump_unstable_fire_ice (Utilities|Casting|CastToB_Shotgun_Pump_Unstable_FireIce (Variables|/Script/FortniteGame.FortPawn|GetCurrentWeapon (Ability|GetActivatingPawn))))
  (return (Variables|/FirePetalWeaponGameplay/Gameplay/FirePetal_PumpShotgun/FireIce/B_Shotgun_Pump_Unstable_FireIce.B_Shotgun_Pump_Unstable_FireIce_C|GetIsIceEffectActive? _asb_shotgun_pump_unstable_fire_ice)))

(fn GetHitActorsFromEventData
  (params (GameplayEventData :type Gameplay%20Event%20Data%20Structure%20%28by%20ref%29))
  (bind (_eventtag _instigator _target) (Utilities|Struct|BreakGameplayEventData GameplayEventData))
  (bind _output_get (LocalVariables|SetLocalTarget _target))
  (Utilities|IsValid _output_get
    (:Is%20Valid
      (bind _returnvalue (Default|GetValidAvatarController true)
        (:Is%20Valid
          (if (CallFunction|IsPawnandNotFriendly _returnvalue (LocalVariables|GetLocalTarget))
            (return (Utilities|Array|MakeArray (LocalVariables|GetLocalTarget)))
            (else
              (return 0))))))
    (:Is%20Not%20Valid
      (return 0))))

(fn IsDamgeFromSourceWeapon
  (params (GameplayEventData :type Gameplay%20Event%20Data%20Structure%20%28by%20ref%29))
  (bind (_eventtag _instigator _target _optionalobject _optionalobject2 _contexthandle) (Utilities|Struct|BreakGameplayEventData GameplayEventData))
  (bind _asfort_pawn (CallFunction|GetAvatarasFortPawn))
  (Utilities|IsValid _asfort_pawn
    (:Is%20Valid
      (return (Utilities|Equal(Object) (Variables|/Script/FortniteGame.FortPawn|GetCurrentWeapon _asfort_pawn) (Ability|EffectContext|GetSourceObject _contexthandle))))
    (:Is%20Not%20Valid
      (return false))))

(fn IsPawnAndNotFriendly
  (params (OwningController :type Actor%20Object%20Reference) (HitActor :type Actor%20Object%20Reference))
  (return (and (Utilities|Enum|Equal(Enum) (Team|GetTeamAffiliation OwningController HitActor) "Hostile") (CallFunction|CanActivateProcfromPawn HitActor))))

(fn SetCurrentEffectFromParent
  (params (SetIsIceNext :type Boolean))
  (bind _asb_shotgun_pump_unstable_fire_ice (Utilities|Casting|CastToB_Shotgun_Pump_Unstable_FireIce (Variables|/Script/FortniteGame.FortPawn|GetCurrentWeapon (Ability|GetActivatingPawn))))
  (Variables|/FirePetalWeaponGameplay/Gameplay/FirePetal_PumpShotgun/FireIce/B_Shotgun_Pump_Unstable_FireIce.B_Shotgun_Pump_Unstable_FireIce_C|SetIsIceEffectActive? SetIsIceNext _asb_shotgun_pump_unstable_fire_ice))
