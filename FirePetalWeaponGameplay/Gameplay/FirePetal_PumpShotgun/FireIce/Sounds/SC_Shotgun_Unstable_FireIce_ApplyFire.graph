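# Sound cue for the fire-apply effect of the Unstable FireIce pump shotgun in the FirePetal weapon plugin.
# After a short fixed delay, a modulator varies one of four equally weighted random wave variations.
# It plays at reduced volume in the weapon foley class with applied-VFX attenuation.
# Concurrency allows one instance per owner and blocks new plays rather than cutting off the current one.
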SoundCue /FirePetalWeaponGameplay/Gameplay/FirePetal_PumpShotgun/FireIce/Sounds/SC_Shotgun_Unstable_FireIce_ApplyFire
root=Delay_0
nodes [
 {id=Delay_0, children [Modulator_0], properties {delayMax=0.2, delayMin=0.2}}
 {id=Modulator_0, children [Random_0]}
 {id=Random_0, children [WavePlayer_0,WavePlayer_1,WavePlayer_2,WavePlayer_3], properties {weights [1,1,1,1]}}
 {
  id=WavePlayer_0
  properties {soundWaveAssetPtr=&/Shotgun_Unstable_FireIce_ApplyFire_04}
 }
 {
  id=WavePlayer_1
  properties {soundWaveAssetPtr=&/Shotgun_Unstable_FireIce_ApplyFire_03}
 }
 {
  id=WavePlayer_2
  properties {soundWaveAssetPtr=&/Shotgun_Unstable_FireIce_ApplyFire_02}
 }
 {
  id=WavePlayer_3
  properties {soundWaveAssetPtr=&/Shotgun_Unstable_FireIce_ApplyFire_01}
 }
]
properties {
 attenuationSettings=/Game/Sounds/Attenuation/ATT_Weapon_VFX_Applied
 maxDistance=1300
 soundClassObject=&Fort_WeaponFoley
 volumeMultiplier=0.5
 concurrencyOverrides {bLimitToOwner=true, maxCount=1, resolutionRule=PreventNew}
}
